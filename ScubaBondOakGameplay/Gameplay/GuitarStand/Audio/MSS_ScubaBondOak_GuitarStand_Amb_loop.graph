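# Mono MetaSound Source for the looping ambient sound of the ScubaBondOak guitar stand.
# On play, a single Wave Player feeds the guitar stand loop wave straight to the output; the exposed array of Ability Ready waves is left unconnected.
# Item ambient attenuation spatializes it, overridden concurrency allows one instance per owner without replacing it, and a Start control bus modulates volume.
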MetaSoundSource /ScubaBondOakGameplay/Gameplay/GuitarStand/Audio/MSS_ScubaBondOak_GuitarStand_Amb_loop
inputs [
 UE.Source.OnPlay {type=Trigger}
 Wave Asset {type=WaveAsset:Array}
]
interfaces [UE.OutputFormat.Mono:1.0, UE.Source:1.1]
outputs ["UE.OutputFormat.Mono.Audio:0" {type=Audio}]
pages [Default {
  input_defaults {
   UE.Source.OnPlay=null
   Wave Asset [
    /ScubaBondOakGameplay/Assets/Audio/SFX/wav/ScubaBondOak_Ability_Ready_01
    /ScubaBondOakGameplay/Assets/Audio/SFX/wav/ScubaBondOak_Ability_Ready_02
   ]
  }
  nodes [WavePlayer_0 = UE.Wave Player:Mono(Play=^UE.Source.OnPlay) {also {defaults {Loop="<MetaSoundBuilderResult.SUCCEEDED: 0>", Wave Asset=&/ScubaBond_Oak_GuitarStand_Amb_lp}}}]
  links [{dst="@UE.OutputFormat.Mono.Audio:0", src=WavePlayer_0.Out Mono}]
 }]
properties {
 assetImportData=&self:AssetImportData
 attenuationSettings=/Game/Sounds/Attenuation/Fort_Item_Ambient_Loop_Mono_Default_Att
 bLooping=true
 bOverrideConcurrency=true
 duration=10000
 maxDistance=1250
 sampleRate=48000
 soundClassObject=/Game/Sounds/FortSoundClassesAndModes/Fort_Loot_Ambient_Loop
 volume=1.3
 concurrencyOverrides {bLimitToOwner=true, maxCount=1, resolutionRule=PreventNew, retriggerTime=0.2}
 concurrencySet [/ScubaBondOakGameplay/Assets/Audio/SFX/Conc_ScubaBondOak_Abilities]
 modulationSettings {
  volumeRouting=Override
  volumeModulationDestination {bEnableModulation=true, modulators [/ScubaBondOakGameplay/Assets/Audio/SFX/CB_ScubaBondOak_Start]}
 }
}
